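# RollbackUI is a tool widget for browsing and searching CVar rollbacks, with a warning shown in Shipping builds.
# It loads rollbacks by category or Jira ticket, sorts them, and fills a category dropdown, which text can filter in category mode.
# Each selection or search rebuilds a RollbackUI_Category_Ingame child with matching CVars, searching name, category, Jira and description.
# Up to ten recent searches persist through a SaveGame slot and appear as clickable buttons.
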WidgetBlueprint /Game/ArtTools/Tech_Art/Rollbacks_UI/RollbackUI
graphs [
 EventGraph {nodes [
   n1 = "On Check State Changed (ShowByJira)"!() -> (OutputDelegate, then, bIsChecked:bool) {is=title K2Node_ComponentBoundEvent}
   > n2 = Branch(Condition=@n1.bIsChecked)
   > n3 = GetRollbacksByJiraTicket!(execute, JiraTicket:String) -> (then, ReturnValue:Array of Cvar Rollback Structures)
   > n4 = Set AllRollbacks(AllRollbacks:Array of Cvar Rollback Structures=@n3) -> (Output_Get:Array of Cvar Rollback Structures)
   n5 = "Get  Categories"() -> (" Categories":Array of Names)
   > Clear(TargetArray:Array of Names=@n5)
   > n6 = For Each Loop!(Exec, Array:Array of Cvar Rollback Structures=@n4) -> (
    LoopBody
    Array Element_RollbackCvar:Name
    Array Element_State:bool=~
    Array Element_Category:Name
    Array Element_Jiras:String
    Array Element_Description:String
    Array Index
    Completed
   )
   > n7 = For Each Loop!(Exec, Array:Array of Strings=ParseIntoArray(SourceString=@n6.Array Element_Jiras, Delimiter=" ", CullEmptyStrings=false)) -> (LoopBody, Array Element:String, Array Index, Completed)
   > Add Unique(TargetArray:Array of Names=@n5, NewItem:"Name (by ref)"=String To Name(InString=@n7))
   n8 = SortRollbacks!(
    execute=[@n16.Completed, @n6.Completed]
    self
    UnsortedRollbacks:Array of Cvar Rollback Structures=Get AllRollbacks() -> (AllRollbacks:Array of Cvar Rollback Structures)
   ) -> (then, SortedRollbacks:Array of Cvar Rollback Structures)
   > Set AllRollbacks(AllRollbacks=@n8)
   > n9 = SortCategories!(execute, self, Unsorted:Array of Names="Get  Categories"()) -> (then, Sorted:Array of Names)
   > "Set  Categories"(" Categories":Array of Names=@n9) -> (Output_Get:Array of Names)
   n10 = Get CategorySelector() -> (CategorySelector:"Combo Box (String) Object Reference")
   > ClearOptions(self:"Combo Box (String) Object Reference"=@n10)
   n11 = "Get  Categories"()
   > n12 = For Each Loop(Array:Array of Names=@n11) -> (Array Element:Name)
   n13 = "To String (Name)"(InName="Get (a copy)"(Array:Array of Names=@n11) -> (Output:Name))
   SetSelectedOption!(execute=@n12.Completed, self:"Combo Box (String) Object Reference"=@n10, Option:String=@n13) -> (then)
   > Switch Combo Box!(execute, self, ComboBoxValue:String=@n13, UseSearch?:bool=false) -> (then)
   AddOption!(execute=@n12.LoopBody, self:"Combo Box (String) Object Reference"=@n10, Option:String="To String (Name)"(InName=@n12)) -> (then)
   n14 = GetRollbacksByCategories!(execute=[@n25.then, @n22.then, @n2.else], Category:String) -> (then, ReturnValue:Array of Cvar Rollback Structures)
   > n15 = Set AllRollbacks(AllRollbacks=@n14)
   > Clear(TargetArray="Get  Categories"())
   > n16 = For Each Loop!(Exec, Array:Array of Cvar Rollback Structures=@n15) -> (
    LoopBody
    Array Element_RollbackCvar:Name
    Array Element_State:bool=~
    Array Element_Category:Name
    Array Element_Jiras:String
    Array Element_Description:String
    Array Index
    Completed
   )
   > K2Node_Composite_0 = FilterCategory!(execute, Element:Name=@n16.Array Element_Category) -> (then:Exec) {is=title K2Node_Composite}
   n17 = "On Selection Changed (CategorySelector)"!() -> (OutputDelegate, then, SelectedItem:String, SelectionType:ESelectInfo Enum) {is=title K2Node_ComponentBoundEvent}
   > Switch Combo Box(ComboBoxValue=@n17.SelectedItem, UseSearch?=true)
   "On Clicked (PreviousSearch_8)"() {is=title K2Node_ComponentBoundEvent}
   > SearchHistoryPressed!(
    execute
    self
    SearchText:"Text (by ref)"=Get Text(self:Text Object Reference=Get PreviousSearchText_8() -> (PreviousSearchText_8:Common Text Object Reference))
   ) -> (then)
   "On Clicked (PreviousSearch_9)"() {is=title K2Node_ComponentBoundEvent}
   > SearchHistoryPressed(SearchText=Get Text(self=Get PreviousSearchText_9() -> (PreviousSearchText_9:Common Text Object Reference)))
   "On Clicked (PreviousSearch_6)"() {is=title K2Node_ComponentBoundEvent}
   > SearchHistoryPressed(SearchText=Get Text(self=Get PreviousSearchText_6() -> (PreviousSearchText_6:Common Text Object Reference)))
   n18 = "OnTextChanged (Editable Text) (EditableText_CategorySearchBox)"!() -> (OutputDelegate, then, Text:"Text (by ref)") {is=title K2Node_ComponentBoundEvent}
   > Set Filter(Filter:Text=@n18.Text) -> (Output_Get:Text)
   > Rework()
   n19 = "OnTextCommitted (Editable Text) (SearchBar)"!() -> (OutputDelegate, then, Text:"Text (by ref)", CommitMethod:ETextCommit Enum=Default) {is=title K2Node_ComponentBoundEvent}
   > n20 = Branch(Condition=TextIsEmpty(InText=@n19.Text))
   Switch Combo Box(
    execute=[@n21.then, @n20.then]
    ComboBoxValue=GetSelectedOption!(self:"Combo Box (String) Object Reference"=Get CategorySelector()) -> (ReturnValue:String)
    UseSearch?=true
   )
   n21 = UpdateSearchhistory!(execute=@n20.else, self, Search:String="To String (Text)"(InText=@n19.Text)) -> (then)
   "On Clicked (PreviousSearch)"() {is=title K2Node_ComponentBoundEvent}
   > SearchHistoryPressed(SearchText=Get Text(self=Get PreviousSearchText() -> (PreviousSearchText:Common Text Object Reference)))
   "On Clicked (PreviousSearch_1)"() {is=title K2Node_ComponentBoundEvent}
   > SearchHistoryPressed(SearchText=Get Text(self=Get PreviousSearchText_1() -> (PreviousSearchText_1:Common Text Object Reference)))
   "On Clicked (PreviousSearch_2)"() {is=title K2Node_ComponentBoundEvent}
   > SearchHistoryPressed(SearchText=Get Text(self=Get PreviousSearchText_2() -> (PreviousSearchText_2:Common Text Object Reference)))
   "On Clicked (PreviousSearch_3)"() {is=title K2Node_ComponentBoundEvent}
   > SearchHistoryPressed(SearchText=Get Text(self=Get PreviousSearchText_3() -> (PreviousSearchText_3:Common Text Object Reference)))
   "On Clicked (PreviousSearch_4)"() {is=title K2Node_ComponentBoundEvent}
   > SearchHistoryPressed(SearchText=Get Text(self=Get PreviousSearchText_4() -> (PreviousSearchText_4:Common Text Object Reference)))
   "On Clicked (PreviousSearch_5)"() {is=title K2Node_ComponentBoundEvent}
   > SearchHistoryPressed(SearchText=Get Text(self=Get PreviousSearchText_5() -> (PreviousSearchText_5:Common Text Object Reference)))
   "On Clicked (PreviousSearch_7)"() {is=title K2Node_ComponentBoundEvent}
   > SearchHistoryPressed(SearchText=Get Text(self=Get PreviousSearchText_7() -> (PreviousSearchText_7:Common Text Object Reference)))
   n22 = Rework() {is=title K2Node_CustomEvent}
   Event Construct()
   > n23 = Branch(Condition="Equal, Case Insensitive (String)"(A=GetBuildConfiguration!() -> (ReturnValue:String), B=Shipping))
   > n24 = Set Visibility(Visibility=Visible, self=Get ShippingWarning() -> (ShippingWarning:Common Text Object Reference))
   n25 = LoadSearchHistory(execute=[@n23.else, @n24.then])
  ]}
 FilterCategory {
  path=EventGraph.K2Node_Composite_0.FilterCategory
  nodes [
   n1 = Inputs!() -> (execute, Element:Name)
   n2 = "To String (Text)"(InText=Get Filter() -> (Filter:Text))
   > n3 = Branch(Condition=IsEmpty(InString=@n2))
   n4 = Add Unique(execute=[@n3.then, @n5.then], TargetArray="Get  Categories"(), NewItem=@n1)
   Outputs!(then:Exec=[@n4.then, @n5.else])
   n5 = Branch(execute=@n3.else, Condition=Contains(SearchIn="To String (Name)"(InName=@n1), Substring=@n2))
  ]
 }
 LoadSearchHistory {nodes [
   LoadSearchHistory() {is=title K2Node_FunctionEntry}
   n1 = Get Slot Name() -> (Slot Name:String)
   > n2 = DoesSaveGameExist!(execute, SlotName:String=@n1, UserIndex:int=0) -> (then, ReturnValue:bool)
   > Branch(Condition=@n2)
   > n3 = LoadGameFromSlot!(execute, SlotName:String=@n1, UserIndex:int=0) -> (then, ReturnValue:Save Game Object Reference)
   > n4 = Set SearchHistorySG(SearchHistorySG:Save Game Object Reference=@n3) -> (Output_Get:Save Game Object Reference)
   > n5 = Cast To RollbackUI_Searchhistory_SG!(execute, Object=@n4) -> (then, CastFailed, AsRollback UI Searchhistory SG:Rollback UI Searchhistory SG Object Reference) {is=title K2Node_DynamicCast}
   > Set SearchHistory(SearchHistory:Array of Strings=Get SearchHistory!(self:Rollback UI Searchhistory SG Object Reference=@n5) -> (SearchHistory:Array of Strings)) -> (Output_Get:Array of Strings)
   > UpdateSearchButtons()
  ]}
 SaveSearchHistory {nodes [
   SaveSearchHistory() {is=title K2Node_FunctionEntry}
   n1 = Get SearchHistorySG() -> (SearchHistorySG:Save Game Object Reference)
   > n2 = Is Valid(InputObject=@n1)
   n3 = CreateSaveGameObject!(
    execute=@n2.Is Not Valid
    SaveGameClass:Save Game Class Reference=/Game/ArtTools/Tech_Art/Rollbacks_UI/Data/RollbackUI_Searchhistory_SG.RollbackUI_Searchhistory_SG_C
   ) -> (then, ReturnValue:Rollback UI Searchhistory SG Object Reference)
   > n4 = Set SearchHistorySG(SearchHistorySG=@n3)
   n5 = Cast To RollbackUI_Searchhistory_SG(execute=[@n2.Is Valid, @n4.then], Object=@n1)
   > Set SearchHistory!(execute, SearchHistory=Get SearchHistory!() -> (SearchHistory), self:Rollback UI Searchhistory SG Object Reference=@n5) -> (then, Output_Get)
   > SaveGameToSlot!(execute, SaveGameObject:Save Game Object Reference=Get SearchHistorySG(), SlotName:String=Get Slot Name(), UserIndex:int=0) -> (then, ReturnValue:bool)
  ]}
 SearchForContainingText {nodes [
   n1 = SearchForContainingText!() -> (then, Cvar_RollbackCvar:Name, Cvar_State:bool=~, Cvar_Category:Name, Cvar_Jiras:String, Cvar_Description:String, SearchText:String) {is=title K2Node_FunctionEntry}
   n2 = "To String (Name)"(InName=@n1)
   n3 = "To String (Name)"(InName=@n1.Cvar_Category)
   > Return Node!(
    execute
    IsSafeToInclude?:bool=OR Boolean!(
     A=Contains(SearchIn=@n2, Substring=@n1.SearchText)
     B=Contains(SearchIn=@n3, Substring=@n1.SearchText)
     C:bool=Contains(SearchIn=@n1.Cvar_Jiras, Substring=@n1.SearchText)
     D:bool=Contains(SearchIn=@n1.Cvar_Description, Substring=@n1.SearchText)
    ) -> (ReturnValue)
   )
  ]}
 SearchHistoryPressed {nodes [
   n1 = SearchHistoryPressed!() -> (then, SearchText:"Text (by ref)") {is=title K2Node_FunctionEntry}
   > Set Text(Text=@n1, self:Editable Text Object Reference=Get SearchBar() -> (SearchBar:Editable Text Object Reference))
   > UpdateSearchhistory(Search="To String (Text)"(InText=@n1))
   > Switch Combo Box(ComboBoxValue=GetSelectedOption(self=Get CategorySelector()), UseSearch?=true)
  ]}
 SearchHistoryUpdateCurrent {nodes [SearchHistoryUpdateCurrent() {is=title K2Node_FunctionEntry}]}
 SortCategories {nodes [
   n1 = SortCategories!() -> (then, Unsorted:Array of Names) {is=title K2Node_FunctionEntry}
   > n2 = For Each Loop!(Exec, Array:Array of Names=@n1) -> (LoopBody, Array Element:Name, Array Index, Completed)
   n3 = Get SortString() -> (SortString:Array of Strings)
   SortStringArray!(execute=@n2.Completed, TargetArray:Array of Strings=@n3, bStableSort:bool=false, SortOrder:EArraySortOrder Enum=Ascending) -> (then)
   > n4 = For Each Loop(Array:Array of Strings=@n3) -> (Array Element:String)
   Return Node!(execute=@n4.Completed, Sorted:Array of Names=Get ReturnName() -> (ReturnName:Array of Names))
   > Add(TargetArray:Array of Names=Get ReturnName(), NewItem:"Name (by ref)"=String To Name(InString=@n4)) {is=title K2Node_CallArrayFunction}
   Add(execute=@n2.LoopBody, TargetArray:Array of Strings=Get SortString(), NewItem:"String (by ref)"="To String (Name)"(InName=@n2))
  ]}
 SortRollbacks {nodes [
   n1 = SortRollbacks!() -> (then, UnsortedRollbacks:Array of Cvar Rollback Structures) {is=title K2Node_FunctionEntry}
   > Resize!(
    execute
    TargetArray:Array of Cvar Rollback Structures=Get ReturnRollbacks() -> (ReturnRollbacks:Array of Cvar Rollback Structures)
    Size:int=Length(TargetArray:Array of Cvar Rollback Structures=@n1)
   ) -> (then) {is=title K2Node_CallArrayFunction}
   > n2 = For Each Loop!(Exec, Array:Array of Cvar Rollback Structures=@n1) -> (
    LoopBody
    Array Element_RollbackCvar:Name
    Array Element_State:bool=~
    Array Element_Category:Name
    Array Element_Jiras:String
    Array Element_Description:String
    Array Index
    Completed
   )
   n3 = Get RollbackStrings() -> (RollbackStrings:Array of Strings)
   SortStringArray(execute=@n2.Completed, TargetArray=@n3, bStableSort=false, SortOrder=Ascending)
   > n4 = For Each Loop!(Exec, Array:Array of Strings=@n3) -> (LoopBody, Array Element:String, Array Index, Completed)
   Return Node!(execute=@n4.Completed, SortedRollbacks:Array of Cvar Rollback Structures=Get ReturnRollbacks())
   n5 = Get RollbackStrings()
   > Set Array Elem!(
    execute
    TargetArray:Array of Cvar Rollback Structures=Get ReturnRollbacks()
    Index:int=@n4.Array Index
    Item:"Cvar Rollback Structure (by ref)"="Get (a copy)"(Array:Array of Cvar Rollback Structures=@n1, Dimension 1=Find Item!(TargetArray=@n5, ItemToFind:"String (by ref)"=@n4) -> (ReturnValue)) -> (Output:Cvar Rollback Structure)
    bSizeToFit:bool=false
   ) -> (then)
   Add(execute=@n2.LoopBody, TargetArray=@n5, NewItem="To String (Name)"(InName=@n2))
  ]}
 Switch Combo Box {nodes [
   n1 = Switch Combo Box!() -> (then, ComboBoxValue:String, UseSearch?:bool=~) {is=title K2Node_FunctionEntry}
   > RemoveFromParent(self=Get CategoryWidgetRef() -> (CategoryWidgetRef:Rollback UI Category Ingame Object Reference))
   > n2 = Create Widget!(
    execute:Exec
    Class:User Widget Class Reference=/Game/ArtTools/Tech_Art/Rollbacks_UI/Data/RollbackUI_Category_Ingame.RollbackUI_Category_Ingame_C
    OwningPlayer:Player Controller Object Reference
    Category:Name=String To Name(InString=@n1)
   ) -> (then:Exec, ReturnValue:Rollback UI Category Ingame Object Reference)
   > AddChild(self=Get ScrollBox_30() -> (ScrollBox_30:Scroll Box Object Reference), Content=@n2)
   > Set CategoryWidgetRef(CategoryWidgetRef:Rollback UI Category Ingame Object Reference=@n2) -> (Output_Get:Rollback UI Category Ingame Object Reference)
   > n3 = For Each Loop(Array:Array of Cvar Rollback Structures=Get AllRollbacks()) -> (Array Element:Cvar Rollback Structure)
   > n4 = Branch(Condition=@n1.UseSearch?)
   n5 = "GetText (Editable Text)"!(self:Editable Text Object Reference=Get SearchBar()) -> (ReturnValue:Text)
   n6 = "To String (Text)"(InText=@n5)
   > n7 = Branch(Condition=OR Boolean!(
     A=SearchForContainingText!(Cvar:Cvar Rollback Structure=@n3, SearchText:String=@n6) -> (IsSafeToInclude?:bool) {is=title K2Node_CallFunction}
     B=TextIsEmpty(InText=@n5)
    ) -> (ReturnValue))
   n8 = Break Cvar Rollback!(CvarRollback:"Cvar Rollback Structure (by ref)"=@n3) -> (RollbackCvar:Name, State:bool=~, Category:Name, Jiras:String, Description:String) {is=title K2Node_BreakStruct}
   n9 = ParseIntoArray(SourceString=@n8.Jiras, Delimiter=" ", CullEmptyStrings=false)
   n10 = "To String (Name)"(InName=@n8.Category)
   Branch(
    execute=[@n4.else, @n7.then]
    Condition=OR Boolean(A="Equal Exactly (String)"(A=@n1, B=@n10), B=Contains Item(TargetArray:Array of Strings=@n9, ItemToFind:"String (by ref)"=@n1))
   )
   > AddNewCvar!(
    execute
    self:Rollback UI Category Ingame Object Reference=Get CategoryWidgetRef()
    Cvar:Name=@n8
    State:bool=@n8.State
    Jiras:String=@n8.Jiras
    Description:String=@n8.Description
   ) -> (then)
  ]}
 UpdateSearchButtons {nodes [
   UpdateSearchButtons() {is=title K2Node_FunctionEntry}
   > n1 = Sequence!(execute) -> (then_0, then_1, then_2:Exec, then_3:Exec, then_4:Exec, then_5:Exec, then_6:Exec, then_7:Exec, then_8:Exec, then_9:Exec)
   n2 = "To Text (String)"(InString="Get (a copy)"(Array:Array of Strings=Get SearchHistory(), Dimension 1=Get LocalCount() -> (LocalCount:int=0)) -> (Output:String))
   > n3 = Set Text(Text=@n2, self:Text Object Reference=Get PreviousSearchText())
   Increment Int(" "=[@n3.then, @n4.then, @n5.then, @n6.then, @n7.then, @n8.then, @n9.then, @n11.then, @n10.then], Value=Get LocalCount() -> (LocalCount=0))
   n4 = Set Text(execute=@n1.then_1, Text=@n2, self=Get PreviousSearchText_1())
   n5 = Set Text(execute=@n1.then_2, Text=@n2, self=Get PreviousSearchText_2())
   n6 = Set Text(execute=@n1.then_3, Text=@n2, self=Get PreviousSearchText_3())
   n7 = Set Text(execute=@n1.then_4, Text=@n2, self=Get PreviousSearchText_4())
   n8 = Set Text(execute=@n1.then_5, Text=@n2, self=Get PreviousSearchText_5())
   n9 = Set Text(execute=@n1.then_6, Text=@n2, self=Get PreviousSearchText_6())
   n10 = Set Text(execute=@n1.then_7, Text=@n2, self=Get PreviousSearchText_7())
   n11 = Set Text(execute=@n1.then_8, Text=@n2, self=Get PreviousSearchText_8())
   Set Text(execute=@n1.then_9, Text=@n2, self=Get PreviousSearchText_9())
  ]}
 UpdateSearchhistory {nodes [
   n1 = UpdateSearchhistory!() -> (then, Search:String) {is=title K2Node_FunctionEntry}
   > n2 = Branch(Condition=Contains Item(TargetArray=Get SearchHistory(), ItemToFind=@n1))
   > Return Node!(execute)
   n3 = Length(TargetArray:Array of Strings=Get SearchHistory())
   n4 = Branch(execute=@n2.else, Condition="integer < integer"(A=@n3, B=10))
   > Add Unique(TargetArray:Array of Strings=Get SearchHistory(), NewItem:"String (by ref)"=@n1)
   > SaveSearchHistory() {is=title K2Node_CallFunction}
   > UpdateSearchButtons() {is=title K2Node_CallFunction}
   > Return Node()
   n5 = For Loop(execute=@n4.else, LastIndex=int - int(A=@n3, B=2))
   n6 = Get SearchHistory()
   Set Array Elem(execute=@n5.Completed, TargetArray:Array of Strings=@n6, Index=0, Item:"String (by ref)"=@n1, bSizeToFit=false)
   > SaveSearchHistory()
   > UpdateSearchButtons()
   > Return Node()
   n7 = int - int(A=int - int(A=@n3), B=@n5)
   Set Array Elem(execute=@n5.LoopBody, TargetArray=@n6, Index=@n7, Item="Get (a copy)"(Array=@n6, Dimension 1=int - int(A=@n7)), bSizeToFit=false)
  ]}
 Was Found in Search {nodes [
   Was Found in Search!() -> (then, SearchTerm:String, NewParam:String) {is=title K2Node_FunctionEntry}
   > Return Node!(execute, IsSafeToInclude?:bool=false)
  ]}
]
variables [
 " Categories" {type="array<string>", default []}
 AllRollbacks {type="array<CvarRollback>", default []}
 CategoryWidgetRef {type=RollbackUI_Category_Ingame_C, default=null}
 Filter {type=string, default=""}
 SearchHistory {type="array<string>", default []}
 SearchHistorySG {type=SaveGame, default=null}
 Slot Name {type=string, default=RollbackUISearchSave}
]
widgets {
 info {inheritedWidgetCount=0, parentClass=&UserWidget, rootWidgetClass=&Overlay, widgetCount=43}
 widgets [
  Overlay_27 {widgetClassPath=&Overlay, widget=Overlay_27}
  HorizontalBox_0 {
   widgetClassPath=&HorizontalBox
   parent=Overlay_27
   widget=HorizontalBox_0
   slot {class=OverlaySlot, properties {horizontalAlignment=HAlign_Fill, verticalAlignment=VAlign_Fill}}
  }
  Border_0 {
   widgetClassPath=&Border
   parent=HorizontalBox_0
   widget=Border_0
   properties {brushColor=<rgba 0.2708 0.2708 0.2708 1>}
   slot {class=HorizontalBoxSlot}
  }
  VerticalBox_0 {widgetClassPath=&VerticalBox, parent=Border_0, widget=VerticalBox_0, slot {class=BorderSlot}}
  CommonTextBlock_34 {
   widgetClassPath=&CommonTextBlock
   parent=VerticalBox_0
   widget=CommonTextBlock_34
   properties {
    style=/FortUILibrary/Styles/Typography/Foundation/TextStyle-Base-XS.TextStyle-Base-XS_C
    text=Search
    font {
     bForceMonospaced=false
     bMaterialIsStencil=false
     fontMaterial=null
     fontObject=&NotoSans
     letterSpacing=0
     monospacedWidth=1
     size=16
     skewAmount=0
     typefaceFontName=NotoSans-Regular
     outlineSettings {
      bApplyOutlineToDropShadows=false
      bMiteredCorners=false
      bSeparateFillAlpha=false
      outlineMaterial=null
      outlineSize=0
      outlineColor=<rgba 0 0 0 1>
     }
    }
    shadowOffset=<0 0>
   }
   slot {class=VerticalBoxSlot}
  }
  UniformGridPanel_0 {
   widgetClassPath=/Script/UMG.UniformGridPanel
   parent=VerticalBox_0
   widget=UniformGridPanel_0
   properties {minDesiredSlotHeight=30, minDesiredSlotWidth=200}
   slot {class=VerticalBoxSlot}
  }
  Border_28 {
   widgetClassPath=&Border
   parent=UniformGridPanel_0
   widget=Border_28
   properties {brushColor=<rgba 0 0 0 1>}
   slot {class=UniformGridSlot, properties {horizontalAlignment=HAlign_Fill}}
  }
  SearchBar {
   widgetClassPath=/Script/UMG.EditableText
   bIsVariable=true
   parent=Border_28
   widget=SearchBar
   properties {
    toolTipText=Type Here To Search
    widgetStyle {
     backgroundImageComposing {
      bIsDynamicallyLoaded=false
      drawAs=Border
      imageType=FullColor
      mirroring=NoMirror
      resourceName=../../../Engine/Content/Slate/Old/HyperlinkDotted.png
      resourceObject=null
      tiling=Both
      imageSize=<0 0>
      margin=<ltrb 0 0 0 0.1875>
      outlineSettings {
       bUseBrushTransparency=false
       roundingType=HalfHeightRadius
       width=0
       color {colorUseRule=UseColor_Specified, specifiedColor=<rgba 0 0 0 0>}
       cornerRadii=<0 0 0 0>
      }
      tintColor {colorUseRule=UseColor_Specified, specifiedColor=<rgba 1 1 1 1>}
      uVRegion {bIsValid=false, max=<0 0>, min=<0 0>}
     }
     backgroundImageSelected {
      bIsDynamicallyLoaded=false
      drawAs=Image
      imageType=NoImage
      mirroring=NoMirror
      resourceName=None
      resourceObject=null
      tiling=NoTile
      imageSize=<0 0>
      margin=<ltrb 0 0 0 0>
      outlineSettings {
       bUseBrushTransparency=false
       roundingType=HalfHeightRadius
       width=0
       color {colorUseRule=UseColor_Specified, specifiedColor=<rgba 0 0 0 0>}
       cornerRadii=<0 0 0 0>
      }
      tintColor {colorUseRule=UseColor_Specified, specifiedColor=<rgba 0 0.162 0.7454 1>}
      uVRegion {bIsValid=false, max=<0 0>, min=<0 0>}
     }
     caretImage {
      bIsDynamicallyLoaded=false
      drawAs=Image
      imageType=NoImage
      mirroring=NoMirror
      resourceName=None
      resourceObject=null
      tiling=NoTile
      imageSize=<0 0>
      margin=<ltrb 0 0 0 0>
      outlineSettings {
       bUseBrushTransparency=false
       roundingType=HalfHeightRadius
       width=0
       color {colorUseRule=UseColor_Specified, specifiedColor=<rgba 0 0 0 0>}
       cornerRadii=<0 0 0 0>
      }
      tintColor {colorUseRule=UseColor_Specified, specifiedColor=<rgba 1 1 1 1>}
      uVRegion {bIsValid=false, max=<0 0>, min=<0 0>}
     }
     colorAndOpacity {colorUseRule=UseColor_Foreground, specifiedColor=<rgba 1 0 1 1>}
     font {
      bForceMonospaced=false
      bMaterialIsStencil=false
      fontMaterial=null
      fontObject=&Roboto
      letterSpacing=0
      monospacedWidth=1
      size=24
      skewAmount=0
      typefaceFontName=Regular
      outlineSettings {
       bApplyOutlineToDropShadows=false
       bMiteredCorners=false
       bSeparateFillAlpha=false
       outlineMaterial=null
       outlineSize=0
       outlineColor=<rgba 0 0 0 1>
      }
     }
     hintColorAndOpacity {colorUseRule=UseColor_Foreground, specifiedColor=<rgba 1 0 1 1>}
     selectedBackgroundColor {colorUseRule=UseColor_Foreground, specifiedColor=<rgba 1 0 1 1>}
    }
   }
   slot {class=BorderSlot}
  }
  HorizontalBox_2 {widgetClassPath=&HorizontalBox, parent=VerticalBox_0, widget=HorizontalBox_2, slot {class=VerticalBoxSlot}}
  ShowByJira {
   widgetClassPath=/Script/UMG.CheckBox
   bIsVariable=true
   parent=HorizontalBox_2
   widget=ShowByJira
   properties {widgetStyle {
     checkBoxType=CheckBox
     backgroundHoveredImage {
      bIsDynamicallyLoaded=false
      drawAs=RoundedBox
      imageType=NoImage
      mirroring=NoMirror
      resourceName=None
      resourceObject=null
      tiling=NoTile
      imageSize=<18 18>
      margin=<ltrb 0 0 0 0>
      outlineSettings {
       bUseBrushTransparency=false
       roundingType=FixedRadius
       width=1
       color {colorUseRule=UseColor_ColorTable, specifiedColor=<rgba 0 0 0 0>}
       cornerRadii=<3 3 3 3>
      }
      tintColor {colorUseRule=UseColor_Specified, specifiedColor=<rgba 0.0048 0.0048 0.0048 1>}
      uVRegion {bIsValid=false, max=<0 0>, min=<0 0>}
     }
     backgroundImage {
      bIsDynamicallyLoaded=false
      drawAs=RoundedBox
      imageType=NoImage
      mirroring=NoMirror
      resourceName=None
      resourceObject=null
      tiling=NoTile
      imageSize=<18 18>
      margin=<ltrb 0 0 0 0>
      outlineSettings {
       bUseBrushTransparency=false
       roundingType=FixedRadius
       width=1
       color {colorUseRule=UseColor_ColorTable, specifiedColor=<rgba 0 0 0 0>}
       cornerRadii=<3 3 3 3>
      }
      tintColor {colorUseRule=UseColor_Specified, specifiedColor=<rgba 0.0048 0.0048 0.0048 1>}
      uVRegion {bIsValid=false, max=<0 0>, min=<0 0>}
     }
     backgroundPressedImage {
      bIsDynamicallyLoaded=false
      drawAs=RoundedBox
      imageType=NoImage
      mirroring=NoMirror
      resourceName=None
      resourceObject=null
      tiling=NoTile
      imageSize=<18 18>
      margin=<ltrb 0 0 0 0>
      outlineSettings {
       bUseBrushTransparency=false
       roundingType=FixedRadius
       width=1
       color {colorUseRule=UseColor_ColorTable, specifiedColor=<rgba 0 0 0 0>}
       cornerRadii=<3 3 3 3>
      }
      tintColor {colorUseRule=UseColor_Specified, specifiedColor=<rgba 0.0048 0.0048 0.0048 1>}
      uVRegion {bIsValid=false, max=<0 0>, min=<0 0>}
     }
     borderBackgroundColor {colorUseRule=UseColor_Specified, specifiedColor=<rgba 1 1 1 1>}
     checkedForeground {colorUseRule=UseColor_Specified, specifiedColor=<rgba 1 1 1 1>}
     checkedHoveredForeground {colorUseRule=UseColor_Specified, specifiedColor=<rgba 1 1 1 1>}
     checkedHoveredImage {
      bIsDynamicallyLoaded=false
      drawAs=Image
      imageType=Vector
      mirroring=NoMirror
      resourceName=../../../Engine/Content/Slate/Starship/CoreWidgets/CheckBox/check.svg
      resourceObject=null
      tiling=NoTile
      imageSize=<16 16>
      margin=<ltrb 0 0 0 0>
      outlineSettings {
       bUseBrushTransparency=false
       roundingType=HalfHeightRadius
       width=0
       color {colorUseRule=UseColor_Specified, specifiedColor=<rgba 0 0 0 0>}
       cornerRadii=<0 0 0 0>
      }
      tintColor {colorUseRule=UseColor_Specified, specifiedColor=<rgba 0.0044 0.2384 1 1>}
      uVRegion {bIsValid=false, max=<0 0>, min=<0 0>}
     }
     checkedImage {
      bIsDynamicallyLoaded=false
      drawAs=Image
      imageType=Vector
      mirroring=NoMirror
      resourceName=../../../Engine/Content/Slate/Starship/CoreWidgets/CheckBox/check.svg
      resourceObject=null
      tiling=NoTile
      imageSize=<16 16>
      margin=<ltrb 0 0 0 0>
      outlineSettings {
       bUseBrushTransparency=false
       roundingType=HalfHeightRadius
       width=0
       color {colorUseRule=UseColor_Specified, specifiedColor=<rgba 0 0 0 0>}
       cornerRadii=<0 0 0 0>
      }
      tintColor {colorUseRule=UseColor_Specified, specifiedColor=<rgba 0 0.162 0.7454 1>}
      uVRegion {bIsValid=false, max=<0 0>, min=<0 0>}
     }
     checkedPressedForeground {colorUseRule=UseColor_Specified, specifiedColor=<rgba 1 1 1 1>}
     checkedPressedImage {
      bIsDynamicallyLoaded=false
      drawAs=Image
      imageType=Vector
      mirroring=NoMirror
      resourceName=../../../Engine/Content/Slate/Starship/CoreWidgets/CheckBox/check.svg
      resourceObject=null
      tiling=NoTile
      imageSize=<16 16>
      margin=<ltrb 0 0 0 0>
      outlineSettings {
       bUseBrushTransparency=false
       roundingType=HalfHeightRadius
       width=0
       color {colorUseRule=UseColor_Specified, specifiedColor=<rgba 0 0 0 0>}
       cornerRadii=<0 0 0 0>
      }
      tintColor {colorUseRule=UseColor_Specified, specifiedColor=<rgba 0 0.0802 0.3515 1>}
      uVRegion {bIsValid=false, max=<0 0>, min=<0 0>}
     }
     checkedSlateSound {resourceObject=null}
     foregroundColor {colorUseRule=UseColor_Specified, specifiedColor=<rgba 1 1 1 1>}
     hoveredForeground {colorUseRule=UseColor_Specified, specifiedColor=<rgba 1 1 1 1>}
     hoveredSlateSound {resourceObject=null}
     padding=<ltrb 2 0 0 0>
     pressedForeground {colorUseRule=UseColor_Specified, specifiedColor=<rgba 1 1 1 1>}
     uncheckedHoveredImage {
      bIsDynamicallyLoaded=false
      drawAs=NoDrawType
      imageType=NoImage
      mirroring=NoMirror
      resourceName=None
      resourceObject=null
      tiling=NoTile
      imageSize=<0 0>
      margin=<ltrb 0 0 0 0>
      outlineSettings {
       bUseBrushTransparency=false
       roundingType=HalfHeightRadius
       width=0
       color {colorUseRule=UseColor_Specified, specifiedColor=<rgba 0 0 0 0>}
       cornerRadii=<0 0 0 0>
      }
      tintColor {colorUseRule=UseColor_Specified, specifiedColor=<rgba 1 1 1 1>}
      uVRegion {bIsValid=false, max=<0 0>, min=<0 0>}
     }
     uncheckedImage {
      bIsDynamicallyLoaded=false
      drawAs=NoDrawType
      imageType=NoImage
      mirroring=NoMirror
      resourceName=None
      resourceObject=null
      tiling=NoTile
      imageSize=<0 0>
      margin=<ltrb 0 0 0 0>
      outlineSettings {
       bUseBrushTransparency=false
       roundingType=HalfHeightRadius
       width=0
       color {colorUseRule=UseColor_Specified, specifiedColor=<rgba 0 0 0 0>}
       cornerRadii=<0 0 0 0>
      }
      tintColor {colorUseRule=UseColor_Specified, specifiedColor=<rgba 1 1 1 1>}
      uVRegion {bIsValid=false, max=<0 0>, min=<0 0>}
     }
     uncheckedPressedImage {
      bIsDynamicallyLoaded=false
      drawAs=NoDrawType
      imageType=NoImage
      mirroring=NoMirror
      resourceName=None
      resourceObject=null
      tiling=NoTile
      imageSize=<0 0>
      margin=<ltrb 0 0 0 0>
      outlineSettings {
       bUseBrushTransparency=false
       roundingType=HalfHeightRadius
       width=0
       color {colorUseRule=UseColor_Specified, specifiedColor=<rgba 0 0 0 0>}
       cornerRadii=<0 0 0 0>
      }
      tintColor {colorUseRule=UseColor_Specified, specifiedColor=<rgba 1 1 1 1>}
      uVRegion {bIsValid=false, max=<0 0>, min=<0 0>}
     }
     uncheckedSlateSound {resourceObject=null}
     undeterminedForeground {colorUseRule=UseColor_Specified, specifiedColor=<rgba 1 1 1 1>}
     undeterminedHoveredImage {
      bIsDynamicallyLoaded=false
      drawAs=Image
      imageType=Vector
      mirroring=NoMirror
      resourceName=../../../Engine/Content/Slate/Starship/CoreWidgets/CheckBox/indeterminate.svg
      resourceObject=null
      tiling=NoTile
      imageSize=<16 16>
      margin=<ltrb 0 0 0 0>
      outlineSettings {
       bUseBrushTransparency=false
       roundingType=HalfHeightRadius
       width=0
       color {colorUseRule=UseColor_Specified, specifiedColor=<rgba 0 0 0 0>}
       cornerRadii=<0 0 0 0>
      }
      tintColor {colorUseRule=UseColor_Specified, specifiedColor=<rgba 0.0044 0.2384 1 1>}
      uVRegion {bIsValid=false, max=<0 0>, min=<0 0>}
     }
     undeterminedImage {
      bIsDynamicallyLoaded=false
      drawAs=Image
      imageType=Vector
      mirroring=NoMirror
      resourceName=../../../Engine/Content/Slate/Starship/CoreWidgets/CheckBox/indeterminate.svg
      resourceObject=null
      tiling=NoTile
      imageSize=<16 16>
      margin=<ltrb 0 0 0 0>
      outlineSettings {
       bUseBrushTransparency=false
       roundingType=HalfHeightRadius
       width=0
       color {colorUseRule=UseColor_Specified, specifiedColor=<rgba 0 0 0 0>}
       cornerRadii=<0 0 0 0>
      }
      tintColor {colorUseRule=UseColor_Specified, specifiedColor=<rgba 0 0.162 0.7454 1>}
      uVRegion {bIsValid=false, max=<0 0>, min=<0 0>}
     }
     undeterminedPressedImage {
      bIsDynamicallyLoaded=false
      drawAs=Image
      imageType=Vector
      mirroring=NoMirror
      resourceName=../../../Engine/Content/Slate/Starship/CoreWidgets/CheckBox/indeterminate.svg
      resourceObject=null
      tiling=NoTile
      imageSize=<16 16>
      margin=<ltrb 0 0 0 0>
      outlineSettings {
       bUseBrushTransparency=false
       roundingType=HalfHeightRadius
       width=0
       color {colorUseRule=UseColor_Specified, specifiedColor=<rgba 0 0 0 0>}
       cornerRadii=<0 0 0 0>
      }
      tintColor {colorUseRule=UseColor_Specified, specifiedColor=<rgba 0 0.0802 0.3515 1>}
      uVRegion {bIsValid=false, max=<0 0>, min=<0 0>}
     }
    }}
   slot {class=HorizontalBoxSlot}
  }
  CommonTextBlock_67 {
   widgetClassPath=&CommonTextBlock
   parent=HorizontalBox_2
   widget=CommonTextBlock_67
   properties {
    text=Show By Jira Ticket
    font {
     bForceMonospaced=false
     bMaterialIsStencil=false
     fontMaterial=null
     fontObject=&NotoSans
     letterSpacing=0
     monospacedWidth=1
     size=10
     skewAmount=0
     typefaceFontName=NotoSans-Regular
     outlineSettings {
      bApplyOutlineToDropShadows=false
      bMiteredCorners=false
      bSeparateFillAlpha=false
      outlineMaterial=null
      outlineSize=0
      outlineColor=<rgba 0 0 0 1>
     }
    }
    shadowOffset=<0 0>
   }
   slot {class=HorizontalBoxSlot}
  }
  PreviousFilters {
   widgetClassPath=&VerticalBox
   parent=VerticalBox_0
   widget=PreviousFilters
   slot {class=VerticalBoxSlot, properties {size {sizeRule=Fill, value=1}}}
  }
  PreviousSearch {widgetClassPath=&Button, parent=PreviousFilters, widget=PreviousSearch, slot {class=VerticalBoxSlot}}
  PreviousSearchText {
   widgetClassPath=&CommonTextBlock
   bIsVariable=true
   parent=PreviousSearch
   widget=PreviousSearchText
   properties {
    font {
     bForceMonospaced=false
     bMaterialIsStencil=false
     fontMaterial=null
     fontObject=&NotoSans
     letterSpacing=0
     monospacedWidth=1
     size=12
     skewAmount=0
     typefaceFontName=NotoSans-Regular
     outlineSettings {
      bApplyOutlineToDropShadows=false
      bMiteredCorners=false
      bSeparateFillAlpha=false
      outlineMaterial=null
      outlineSize=0
      outlineColor=<rgba 0 0 0 1>
     }
    }
    shadowOffset=<0 0>
   }
   slot {class=ButtonSlot}
  }
  PreviousSearch_1 {widgetClassPath=&Button, parent=PreviousFilters, widget=PreviousSearch_1, slot {class=VerticalBoxSlot}}
  PreviousSearchText_1 {
   widgetClassPath=&CommonTextBlock
   bIsVariable=true
   parent=PreviousSearch_1
   widget=PreviousSearchText_1
   properties {
    font {
     bForceMonospaced=false
     bMaterialIsStencil=false
     fontMaterial=null
     fontObject=&NotoSans
     letterSpacing=0
     monospacedWidth=1
     size=12
     skewAmount=0
     typefaceFontName=NotoSans-Regular
     outlineSettings {
      bApplyOutlineToDropShadows=false
      bMiteredCorners=false
      bSeparateFillAlpha=false
      outlineMaterial=null
      outlineSize=0
      outlineColor=<rgba 0 0 0 1>
     }
    }
    shadowOffset=<0 0>
   }
   slot {class=ButtonSlot}
  }
  PreviousSearch_2 {widgetClassPath=&Button, parent=PreviousFilters, widget=PreviousSearch_2, slot {class=VerticalBoxSlot}}
  PreviousSearchText_2 {
   widgetClassPath=&CommonTextBlock
   bIsVariable=true
   parent=PreviousSearch_2
   widget=PreviousSearchText_2
   properties {
    font {
     bForceMonospaced=false
     bMaterialIsStencil=false
     fontMaterial=null
     fontObject=&NotoSans
     letterSpacing=0
     monospacedWidth=1
     size=12
     skewAmount=0
     typefaceFontName=NotoSans-Regular
     outlineSettings {
      bApplyOutlineToDropShadows=false
      bMiteredCorners=false
      bSeparateFillAlpha=false
      outlineMaterial=null
      outlineSize=0
      outlineColor=<rgba 0 0 0 1>
     }
    }
    shadowOffset=<0 0>
   }
   slot {class=ButtonSlot}
  }
  PreviousSearch_3 {widgetClassPath=&Button, parent=PreviousFilters, widget=PreviousSearch_3, slot {class=VerticalBoxSlot}}
  PreviousSearchText_3 {
   widgetClassPath=&CommonTextBlock
   bIsVariable=true
   parent=PreviousSearch_3
   widget=PreviousSearchText_3
   properties {
    font {
     bForceMonospaced=false
     bMaterialIsStencil=false
     fontMaterial=null
     fontObject=&NotoSans
     letterSpacing=0
     monospacedWidth=1
     size=12
     skewAmount=0
     typefaceFontName=NotoSans-Regular
     outlineSettings {
      bApplyOutlineToDropShadows=false
      bMiteredCorners=false
      bSeparateFillAlpha=false
      outlineMaterial=null
      outlineSize=0
      outlineColor=<rgba 0 0 0 1>
     }
    }
    shadowOffset=<0 0>
   }
   slot {class=ButtonSlot}
  }
  PreviousSearch_4 {widgetClassPath=&Button, parent=PreviousFilters, widget=PreviousSearch_4, slot {class=VerticalBoxSlot}}
  PreviousSearchText_4 {
   widgetClassPath=&CommonTextBlock
   bIsVariable=true
   parent=PreviousSearch_4
   widget=PreviousSearchText_4
   properties {
    font {
     bForceMonospaced=false
     bMaterialIsStencil=false
     fontMaterial=null
     fontObject=&NotoSans
     letterSpacing=0
     monospacedWidth=1
     size=12
     skewAmount=0
     typefaceFontName=NotoSans-Regular
     outlineSettings {
      bApplyOutlineToDropShadows=false
      bMiteredCorners=false
      bSeparateFillAlpha=false
      outlineMaterial=null
      outlineSize=0
      outlineColor=<rgba 0 0 0 1>
     }
    }
    shadowOffset=<0 0>
   }
   slot {class=ButtonSlot}
  }
  PreviousSearch_5 {widgetClassPath=&Button, parent=PreviousFilters, widget=PreviousSearch_5, slot {class=VerticalBoxSlot}}
  PreviousSearchText_5 {
   widgetClassPath=&CommonTextBlock
   bIsVariable=true
   parent=PreviousSearch_5
   widget=PreviousSearchText_5
   properties {
    font {
     bForceMonospaced=false
     bMaterialIsStencil=false
     fontMaterial=null
     fontObject=&NotoSans
     letterSpacing=0
     monospacedWidth=1
     size=12
     skewAmount=0
     typefaceFontName=NotoSans-Regular
     outlineSettings {
      bApplyOutlineToDropShadows=false
      bMiteredCorners=false
      bSeparateFillAlpha=false
      outlineMaterial=null
      outlineSize=0
      outlineColor=<rgba 0 0 0 1>
     }
    }
    shadowOffset=<0 0>
   }
   slot {class=ButtonSlot}
  }
  PreviousSearch_6 {widgetClassPath=&Button, parent=PreviousFilters, widget=PreviousSearch_6, slot {class=VerticalBoxSlot}}
  PreviousSearchText_6 {
   widgetClassPath=&CommonTextBlock
   bIsVariable=true
   parent=PreviousSearch_6
   widget=PreviousSearchText_6
   properties {
    font {
     bForceMonospaced=false
     bMaterialIsStencil=false
     fontMaterial=null
     fontObject=&NotoSans
     letterSpacing=0
     monospacedWidth=1
     size=12
     skewAmount=0
     typefaceFontName=NotoSans-Regular
     outlineSettings {
      bApplyOutlineToDropShadows=false
      bMiteredCorners=false
      bSeparateFillAlpha=false
      outlineMaterial=null
      outlineSize=0
      outlineColor=<rgba 0 0 0 1>
     }
    }
    shadowOffset=<0 0>
   }
   slot {class=ButtonSlot}
  }
  PreviousSearch_7 {widgetClassPath=&Button, parent=PreviousFilters, widget=PreviousSearch_7, slot {class=VerticalBoxSlot}}
  PreviousSearchText_7 {
   widgetClassPath=&CommonTextBlock
   bIsVariable=true
   parent=PreviousSearch_7
   widget=PreviousSearchText_7
   properties {
    font {
     bForceMonospaced=false
     bMaterialIsStencil=false
     fontMaterial=null
     fontObject=&NotoSans
     letterSpacing=0
     monospacedWidth=1
     size=12
     skewAmount=0
     typefaceFontName=NotoSans-Regular
     outlineSettings {
      bApplyOutlineToDropShadows=false
      bMiteredCorners=false
      bSeparateFillAlpha=false
      outlineMaterial=null
      outlineSize=0
      outlineColor=<rgba 0 0 0 1>
     }
    }
    shadowOffset=<0 0>
   }
   slot {class=ButtonSlot}
  }
  PreviousSearch_8 {widgetClassPath=&Button, parent=PreviousFilters, widget=PreviousSearch_8, slot {class=VerticalBoxSlot}}
  PreviousSearchText_8 {
   widgetClassPath=&CommonTextBlock
   bIsVariable=true
   parent=PreviousSearch_8
   widget=PreviousSearchText_8
   properties {
    font {
     bForceMonospaced=false
     bMaterialIsStencil=false
     fontMaterial=null
     fontObject=&NotoSans
     letterSpacing=0
     monospacedWidth=1
     size=12
     skewAmount=0
     typefaceFontName=NotoSans-Regular
     outlineSettings {
      bApplyOutlineToDropShadows=false
      bMiteredCorners=false
      bSeparateFillAlpha=false
      outlineMaterial=null
      outlineSize=0
      outlineColor=<rgba 0 0 0 1>
     }
    }
    shadowOffset=<0 0>
   }
   slot {class=ButtonSlot}
  }
  PreviousSearch_9 {widgetClassPath=&Button, parent=PreviousFilters, widget=PreviousSearch_9, slot {class=VerticalBoxSlot}}
  PreviousSearchText_9 {
   widgetClassPath=&CommonTextBlock
   bIsVariable=true
   parent=PreviousSearch_9
   widget=PreviousSearchText_9
   properties {
    font {
     bForceMonospaced=false
     bMaterialIsStencil=false
     fontMaterial=null
     fontObject=&NotoSans
     letterSpacing=0
     monospacedWidth=1
     size=12
     skewAmount=0
     typefaceFontName=NotoSans-Regular
     outlineSettings {
      bApplyOutlineToDropShadows=false
      bMiteredCorners=false
      bSeparateFillAlpha=false
      outlineMaterial=null
      outlineSize=0
      outlineColor=<rgba 0 0 0 1>
     }
    }
    shadowOffset=<0 0>
   }
   slot {class=ButtonSlot}
  }
  VerticalBox_1 {
   widgetClassPath=&VerticalBox
   parent=HorizontalBox_0
   widget=VerticalBox_1
   slot {class=HorizontalBoxSlot, properties {verticalAlignment=VAlign_Top, size {sizeRule=Fill, value=1}}}
  }
  ShippingWarning {
   widgetClassPath=&CommonTextBlock
   bIsVariable=true
   parent=VerticalBox_1
   widget=ShippingWarning
   properties {
    autoWrapText=true
    text="Warning: The client is set to shipping.  If the server is also a shipping build then this command will only work on the client."
    visibility=Collapsed
    colorAndOpacity {colorUseRule=UseColor_Specified, specifiedColor=<rgba 1 0 0 1>}
    font {
     bForceMonospaced=false
     bMaterialIsStencil=false
     fontMaterial=null
     fontObject=&NotoSans
     letterSpacing=0
     monospacedWidth=1
     size=24
     skewAmount=0
     typefaceFontName=NotoSans-Regular
     outlineSettings {
      bApplyOutlineToDropShadows=false
      bMiteredCorners=false
      bSeparateFillAlpha=false
      outlineMaterial=null
      outlineSize=0
      outlineColor=<rgba 0 0 0 1>
     }
    }
    shadowOffset=<0 0>
   }
   slot {class=VerticalBoxSlot}
  }
  HorizontalBox_1 {
   widgetClassPath=&HorizontalBox
   parent=VerticalBox_1
   widget=HorizontalBox_1
   slot {class=VerticalBoxSlot, properties {horizontalAlignment=HAlign_Left}}
  }
  AddHere {
   widgetClassPath=&VerticalBox
   bIsVariable=true
   parent=HorizontalBox_1
   widget=AddHere
   slot {class=HorizontalBoxSlot, properties {size {sizeRule=Fill, value=1}}}
  }
  Border_90 {
   widgetClassPath=&Border
   parent=AddHere
   widget=Border_90
   properties {brushColor=<rgba 0.0104 0.0104 0.0104 1>, padding=<ltrb 6 3 6 3>}
   slot {class=VerticalBoxSlot}
  }
  HorizontalBox_3 {
   widgetClassPath=&HorizontalBox
   parent=Border_90
   widget=HorizontalBox_3
   slot {class=BorderSlot, properties {padding=<ltrb 6 3 6 3>}}
  }
  Image_86 {
   widgetClassPath=&Image
   parent=HorizontalBox_3
   widget=Image_86
   properties {brush {
     bIsDynamicallyLoaded=false
     drawAs=Image
     imageType=FullColor
     mirroring=NoMirror
     resourceName=None
     resourceObject=/FortUILibrary/Textures/Icons/T_UI_Icon_UIKit_Search
     tiling=NoTile
     imageSize=<24 24>
     margin=<ltrb 0 0 0 0>
     outlineSettings {
      bUseBrushTransparency=false
      roundingType=HalfHeightRadius
      width=0
      color {colorUseRule=UseColor_Specified, specifiedColor=<rgba 0 0 0 0>}
      cornerRadii=<0 0 0 0>
     }
     tintColor {colorUseRule=UseColor_Specified, specifiedColor=<rgba 1 1 1 1>}
     uVRegion {bIsValid=false, max=<0 0>, min=<0 0>}
    }}
   slot {class=HorizontalBoxSlot, properties {horizontalAlignment=HAlign_Center, verticalAlignment=VAlign_Center}}
  }
  Border_92 {
   widgetClassPath=&Border
   parent=HorizontalBox_3
   widget=Border_92
   properties {brushColor=<rgba 0.002 0.002 0.002 1>}
   slot {class=HorizontalBoxSlot, properties {verticalAlignment=VAlign_Center, padding=<ltrb 5 0 0 0>, size {sizeRule=Fill, value=1}}}
  }
  EditableText_CategorySearchBox {
   widgetClassPath=&EditableText
   bIsVariable=true
   parent=Border_92
   widget=EditableText_CategorySearchBox
   properties {widgetStyle {
     backgroundImageComposing {
      bIsDynamicallyLoaded=false
      drawAs=Border
      imageType=FullColor
      mirroring=NoMirror
      resourceName=../../../Engine/Content/Slate/Old/HyperlinkDotted.png
      resourceObject=null
      tiling=Both
      imageSize=<0 0>
      margin=<ltrb 0 0 0 0.1875>
      outlineSettings {
       bUseBrushTransparency=false
       roundingType=HalfHeightRadius
       width=0
       color {colorUseRule=UseColor_Specified, specifiedColor=<rgba 0 0 0 0>}
       cornerRadii=<0 0 0 0>
      }
      tintColor {colorUseRule=UseColor_Specified, specifiedColor=<rgba 1 1 1 1>}
      uVRegion {bIsValid=false, max=<0 0>, min=<0 0>}
     }
     backgroundImageSelected {
      bIsDynamicallyLoaded=false
      drawAs=Box
      imageType=FullColor
      mirroring=NoMirror
      resourceName=../../../Engine/Content/Slate/Common/EditableTextSelectionBackground.png
      resourceObject=null
      tiling=NoTile
      imageSize=<0 0>
      margin=<ltrb 0.25 0.25 0.25 0.25>
      outlineSettings {
       bUseBrushTransparency=false
       roundingType=HalfHeightRadius
       width=0
       color {colorUseRule=UseColor_Specified, specifiedColor=<rgba 0 0 0 0>}
       cornerRadii=<0 0 0 0>
      }
      tintColor {colorUseRule=UseColor_Specified, specifiedColor=<rgba 1 1 1 1>}
      uVRegion {bIsValid=false, max=<0 0>, min=<0 0>}
     }
     caretImage {
      bIsDynamicallyLoaded=false
      drawAs=Image
      imageType=FullColor
      mirroring=NoMirror
      resourceName=../../../Engine/Content/Slate/Old/White.png
      resourceObject=null
      tiling=NoTile
      imageSize=<16 16>
      margin=<ltrb 0 0 0 0>
      outlineSettings {
       bUseBrushTransparency=false
       roundingType=HalfHeightRadius
       width=0
       color {colorUseRule=UseColor_Specified, specifiedColor=<rgba 0 0 0 0>}
       cornerRadii=<0 0 0 0>
      }
      tintColor {colorUseRule=UseColor_Specified, specifiedColor=<rgba 1 1 1 1>}
      uVRegion {bIsValid=false, max=<0 0>, min=<0 0>}
     }
     colorAndOpacity {colorUseRule=UseColor_Foreground, specifiedColor=<rgba 1 0 1 1>}
     font {
      bForceMonospaced=false
      bMaterialIsStencil=false
      fontMaterial=null
      fontObject=&Roboto
      letterSpacing=0
      monospacedWidth=1
      size=18
      skewAmount=0
      typefaceFontName=Regular
      outlineSettings {
       bApplyOutlineToDropShadows=false
       bMiteredCorners=false
       bSeparateFillAlpha=false
       outlineMaterial=null
       outlineSize=0
       outlineColor=<rgba 0 0 0 1>
      }
     }
     hintColorAndOpacity {colorUseRule=UseColor_Foreground, specifiedColor=<rgba 1 0 1 1>}
     selectedBackgroundColor {colorUseRule=UseColor_Foreground, specifiedColor=<rgba 1 0 1 1>}
    }}
   slot {class=BorderSlot}
  }
  CategorySelector {
   widgetClassPath=/Script/UMG.ComboBoxString
   bIsVariable=true
   parent=AddHere
   widget=CategorySelector
   properties {
    foregroundColor {colorUseRule=UseColor_Specified, specifiedColor=<rgba 0.5271 0.5271 0.5271 1>}
    itemStyle {
     bUseParentRowBrush=false
     activeBrush {
      bIsDynamicallyLoaded=false
      drawAs=Image
      imageType=NoImage
      mirroring=NoMirror
      resourceName=None
      resourceObject=null
      tiling=NoTile
      imageSize=<0 0>
      margin=<ltrb 0 0 0 0>
      outlineSettings {
       bUseBrushTransparency=false
       roundingType=HalfHeightRadius
       width=0
       color {colorUseRule=UseColor_Specified, specifiedColor=<rgba 0 0 0 0>}
       cornerRadii=<0 0 0 0>
      }
      tintColor {colorUseRule=UseColor_Specified, specifiedColor=<rgba 0 0.162 0.7454 1>}
      uVRegion {bIsValid=false, max=<0 0>, min=<0 0>}
     }
     activeHighlightedBrush {
      bIsDynamicallyLoaded=false
      drawAs=Image
      imageType=NoImage
      mirroring=NoMirror
      resourceName=None
      resourceObject=null
      tiling=NoTile
      imageSize=<0 0>
      margin=<ltrb 0 0 0 0>
      outlineSettings {
       bUseBrushTransparency=false
       roundingType=HalfHeightRadius
       width=0
       color {colorUseRule=UseColor_Specified, specifiedColor=<rgba 0 0 0 0>}
       cornerRadii=<0 0 0 0>
      }
      tintColor {colorUseRule=UseColor_Specified, specifiedColor=<rgba 0.0252 0.0319 0.0423 1>}
      uVRegion {bIsValid=false, max=<0 0>, min=<0 0>}
     }
     activeHoveredBrush {
      bIsDynamicallyLoaded=false
      drawAs=Image
      imageType=NoImage
      mirroring=NoMirror
      resourceName=None
      resourceObject=null
      tiling=NoTile
      imageSize=<0 0>
      margin=<ltrb 0 0 0 0>
      outlineSettings {
       bUseBrushTransparency=false
       roundingType=HalfHeightRadius
       width=0
       color {colorUseRule=UseColor_Specified, specifiedColor=<rgba 0 0 0 0>}
       cornerRadii=<0 0 0 0>
      }
      tintColor {colorUseRule=UseColor_Specified, specifiedColor=<rgba 0 0.162 0.7454 1>}
      uVRegion {bIsValid=false, max=<0 0>, min=<0 0>}
     }
     dropIndicator_Above {
      bIsDynamicallyLoaded=false
      drawAs=Box
      imageType=FullColor
      mirroring=NoMirror
      resourceName=../../../Engine/Content/Slate/Common/DropZoneIndicator_Above.png
      resourceObject=null
      tiling=NoTile
      imageSize=<0 0>
      margin=<ltrb 0.625 0.625 0 0>
      outlineSettings {
       bUseBrushTransparency=false
       roundingType=HalfHeightRadius
       width=0
       color {colorUseRule=UseColor_Specified, specifiedColor=<rgba 0 0 0 0>}
       cornerRadii=<0 0 0 0>
      }
      tintColor {colorUseRule=UseColor_Specified, specifiedColor=<rgba 0.0091 0.3515 0.9647 1>}
      uVRegion {bIsValid=false, max=<0 0>, min=<0 0>}
     }
     dropIndicator_Below {
      bIsDynamicallyLoaded=false
      drawAs=Box
      imageType=FullColor
      mirroring=NoMirror
      resourceName=../../../Engine/Content/Slate/Common/DropZoneIndicator_Below.png
      resourceObject=null
      tiling=NoTile
      imageSize=<0 0>
      margin=<ltrb 0.625 0 0 0.625>
      outlineSettings {
       bUseBrushTransparency=false
       roundingType=HalfHeightRadius
       width=0
       color {colorUseRule=UseColor_Specified, specifiedColor=<rgba 0 0 0 0>}
       cornerRadii=<0 0 0 0>
      }
      tintColor {colorUseRule=UseColor_Specified, specifiedColor=<rgba 0.0091 0.3515 0.9647 1>}
      uVRegion {bIsValid=false, max=<0 0>, min=<0 0>}
     }
     dropIndicator_Onto {
      bIsDynamicallyLoaded=false
      drawAs=Box
      imageType=FullColor
      mirroring=NoMirror
      resourceName=../../../Engine/Content/Slate/Common/DropZoneIndicator_Onto.png
      resourceObject=null
      tiling=NoTile
      imageSize=<0 0>
      margin=<ltrb 0.25 0.25 0.25 0.25>
      outlineSettings {
       bUseBrushTransparency=false
       roundingType=HalfHeightRadius
       width=0
       color {colorUseRule=UseColor_Specified, specifiedColor=<rgba 0 0 0 0>}
       cornerRadii=<0 0 0 0>
      }
      tintColor {colorUseRule=UseColor_Specified, specifiedColor=<rgba 0.0091 0.3515 0.9647 1>}
      uVRegion {bIsValid=false, max=<0 0>, min=<0 0>}
     }
     evenRowBackgroundBrush {
      bIsDynamicallyLoaded=false
      drawAs=Image
      imageType=NoImage
      mirroring=NoMirror
      resourceName=None
      resourceObject=null
      tiling=NoTile
      imageSize=<0 0>
      margin=<ltrb 0 0 0 0>
      outlineSettings {
       bUseBrushTransparency=false
       roundingType=HalfHeightRadius
       width=0
       color {colorUseRule=UseColor_Specified, specifiedColor=<rgba 0 0 0 0>}
       cornerRadii=<0 0 0 0>
      }
      tintColor {colorUseRule=UseColor_Specified, specifiedColor=<rgba 0.0103 0.0103 0.0103 1>}
      uVRegion {bIsValid=false, max=<0 0>, min=<0 0>}
     }
     evenRowBackgroundHoveredBrush {
      bIsDynamicallyLoaded=false
      drawAs=Image
      imageType=NoImage
      mirroring=NoMirror
      resourceName=None
      resourceObject=null
      tiling=NoTile
      imageSize=<0 0>
      margin=<ltrb 0 0 0 0>
      outlineSettings {
       bUseBrushTransparency=false
       roundingType=HalfHeightRadius
       width=0
       color {colorUseRule=UseColor_Specified, specifiedColor=<rgba 0 0 0 0>}
       cornerRadii=<0 0 0 0>
      }
      tintColor {colorUseRule=UseColor_Specified, specifiedColor=<rgba 0.0176 0.0176 0.0176 1>}
      uVRegion {bIsValid=false, max=<0 0>, min=<0 0>}
     }
     inactiveBrush {
      bIsDynamicallyLoaded=false
      drawAs=Image
      imageType=NoImage
      mirroring=NoMirror
      resourceName=None
      resourceObject=null
      tiling=NoTile
      imageSize=<0 0>
      margin=<ltrb 0 0 0 0>
      outlineSettings {
       bUseBrushTransparency=false
       roundingType=HalfHeightRadius
       width=0
       color {colorUseRule=UseColor_Specified, specifiedColor=<rgba 0 0 0 0>}
       cornerRadii=<0 0 0 0>
      }
      tintColor {colorUseRule=UseColor_Specified, specifiedColor=<rgba 0.0513 0.0953 0.159 1>}
      uVRegion {bIsValid=false, max=<0 0>, min=<0 0>}
     }
     inactiveHighlightedBrush {
      bIsDynamicallyLoaded=false
      drawAs=Image
      imageType=NoImage
      mirroring=NoMirror
      resourceName=None
      resourceObject=null
      tiling=NoTile
      imageSize=<0 0>
      margin=<ltrb 0 0 0 0>
      outlineSettings {
       bUseBrushTransparency=false
       roundingType=HalfHeightRadius
       width=0
       color {colorUseRule=UseColor_Specified, specifiedColor=<rgba 0 0 0 0>}
       cornerRadii=<0 0 0 0>
      }
      tintColor {colorUseRule=UseColor_Specified, specifiedColor=<rgba 0.0252 0.0319 0.0423 1>}
      uVRegion {bIsValid=false, max=<0 0>, min=<0 0>}
     }
     inactiveHoveredBrush {
      bIsDynamicallyLoaded=false
      drawAs=Image
      imageType=NoImage
      mirroring=NoMirror
      resourceName=None
      resourceObject=null
      tiling=NoTile
      imageSize=<0 0>
      margin=<ltrb 0 0 0 0>
      outlineSettings {
       bUseBrushTransparency=false
       roundingType=HalfHeightRadius
       width=0
       color {colorUseRule=UseColor_Specified, specifiedColor=<rgba 0 0 0 0>}
       cornerRadii=<0 0 0 0>
      }
      tintColor {colorUseRule=UseColor_Specified, specifiedColor=<rgba 0.0513 0.0953 0.159 1>}
      uVRegion {bIsValid=false, max=<0 0>, min=<0 0>}
     }
     oddRowBackgroundBrush {
      bIsDynamicallyLoaded=false
      drawAs=Image
      imageType=NoImage
      mirroring=NoMirror
      resourceName=None
      resourceObject=null
      tiling=NoTile
      imageSize=<0 0>
      margin=<ltrb 0 0 0 0>
      outlineSettings {
       bUseBrushTransparency=false
       roundingType=HalfHeightRadius
       width=0
       color {colorUseRule=UseColor_Specified, specifiedColor=<rgba 0 0 0 0>}
       cornerRadii=<0 0 0 0>
      }
      tintColor {colorUseRule=UseColor_Specified, specifiedColor=<rgba 0.0103 0.0103 0.0103 1>}
      uVRegion {bIsValid=false, max=<0 0>, min=<0 0>}
     }
     oddRowBackgroundHoveredBrush {
      bIsDynamicallyLoaded=false
      drawAs=Image
      imageType=NoImage
      mirroring=NoMirror
      resourceName=None
      resourceObject=null
      tiling=NoTile
      imageSize=<0 0>
      margin=<ltrb 0 0 0 0>
      outlineSettings {
       bUseBrushTransparency=false
       roundingType=HalfHeightRadius
       width=0
       color {colorUseRule=UseColor_Specified, specifiedColor=<rgba 0 0 0 0>}
       cornerRadii=<0 0 0 0>
      }
      tintColor {colorUseRule=UseColor_Specified, specifiedColor=<rgba 0.0176 0.0176 0.0176 1>}
      uVRegion {bIsValid=false, max=<0 0>, min=<0 0>}
     }
     parentRowBackgroundBrush {
      bIsDynamicallyLoaded=false
      drawAs=Image
      imageType=NoImage
      mirroring=NoMirror
      resourceName=None
      resourceObject=null
      tiling=NoTile
      imageSize=<32 32>
      margin=<ltrb 0 0 0 0>
      outlineSettings {
       bUseBrushTransparency=false
       roundingType=HalfHeightRadius
       width=0
       color {colorUseRule=UseColor_Specified, specifiedColor=<rgba 0 0 0 0>}
       cornerRadii=<0 0 0 0>
      }
      tintColor {colorUseRule=UseColor_Specified, specifiedColor=<rgba 1 1 1 1>}
      uVRegion {bIsValid=false, max=<0 0>, min=<0 0>}
     }
     parentRowBackgroundHoveredBrush {
      bIsDynamicallyLoaded=false
      drawAs=Image
      imageType=NoImage
      mirroring=NoMirror
      resourceName=None
      resourceObject=null
      tiling=NoTile
      imageSize=<32 32>
      margin=<ltrb 0 0 0 0>
      outlineSettings {
       bUseBrushTransparency=false
       roundingType=HalfHeightRadius
       width=0
       color {colorUseRule=UseColor_Specified, specifiedColor=<rgba 0 0 0 0>}
       cornerRadii=<0 0 0 0>
      }
      tintColor {colorUseRule=UseColor_Specified, specifiedColor=<rgba 1 1 1 1>}
      uVRegion {bIsValid=false, max=<0 0>, min=<0 0>}
     }
     selectedTextColor {colorUseRule=UseColor_Specified, specifiedColor=<rgba 1 1 1 1>}
     selectorFocusedBrush {
      bIsDynamicallyLoaded=false
      drawAs=Border
      imageType=FullColor
      mirroring=NoMirror
      resourceName=../../../Engine/Content/Slate/Common/Selector.png
      resourceObject=null
      tiling=Both
      imageSize=<0 0>
      margin=<ltrb 0.25 0.25 0.25 0.25>
      outlineSettings {
       bUseBrushTransparency=false
       roundingType=HalfHeightRadius
       width=0
       color {colorUseRule=UseColor_Specified, specifiedColor=<rgba 0 0 0 0>}
       cornerRadii=<0 0 0 0>
      }
      tintColor {colorUseRule=UseColor_Specified, specifiedColor=<rgba 0.701 0.225 0.003 1>}
      uVRegion {bIsValid=false, max=<0 0>, min=<0 0>}
     }
     textColor {colorUseRule=UseColor_Specified, specifiedColor=<rgba 0.5271 0.5271 0.5271 1>}
    }
    widgetStyle {
     comboButtonStyle {
      downArrowAlign=VAlign_Center
      buttonStyle {
       clickedSlateSound {resourceObject=null}
       disabled {
        bIsDynamicallyLoaded=false
        drawAs=RoundedBox
        imageType=NoImage
        mirroring=NoMirror
        resourceName=None
        resourceObject=null
        tiling=NoTile
        imageSize=<0 0>
        margin=<ltrb 0 0 0 0>
        outlineSettings {
         bUseBrushTransparency=false
         roundingType=FixedRadius
         width=0
         color {colorUseRule=UseColor_Specified, specifiedColor=<rgba 0 0 0 0>}
         cornerRadii=<4 4 4 4>
        }
        tintColor {colorUseRule=UseColor_Specified, specifiedColor=<rgba 0.0395 0.0395 0.0395 1>}
        uVRegion {bIsValid=false, max=<0 0>, min=<0 0>}
       }
       disabledForeground {colorUseRule=UseColor_ColorTable, specifiedColor=<rgba 0 0 0 0>}
       hovered {
        bIsDynamicallyLoaded=false
        drawAs=RoundedBox
        imageType=NoImage
        mirroring=NoMirror
        resourceName=None
        resourceObject=null
        tiling=NoTile
        imageSize=<0 0>
        margin=<ltrb 0 0 0 0>
        outlineSettings {
         bUseBrushTransparency=false
         roundingType=FixedRadius
         width=1
         color {colorUseRule=UseColor_ColorTable, specifiedColor=<rgba 0 0 0 0>}
         cornerRadii=<4 4 4 4>
        }
        tintColor {colorUseRule=UseColor_Specified, specifiedColor=<rgba 0.0048 0.0048 0.0048 1>}
        uVRegion {bIsValid=false, max=<0 0>, min=<0 0>}
       }
       hoveredForeground {colorUseRule=UseColor_ColorTable, specifiedColor=<rgba 0 0 0 0>}
       hoveredSlateSound {resourceObject=null}
       normal {
        bIsDynamicallyLoaded=false
        drawAs=RoundedBox
        imageType=NoImage
        mirroring=NoMirror
        resourceName=None
        resourceObject=null
        tiling=NoTile
        imageSize=<0 0>
        margin=<ltrb 0 0 0 0>
        outlineSettings {
         bUseBrushTransparency=false
         roundingType=FixedRadius
         width=1
         color {colorUseRule=UseColor_ColorTable, specifiedColor=<rgba 0 0 0 0>}
         cornerRadii=<4 4 4 4>
        }
        tintColor {colorUseRule=UseColor_Specified, specifiedColor=<rgba 0.0048 0.0048 0.0048 1>}
        uVRegion {bIsValid=false, max=<0 0>, min=<0 0>}
       }
       normalForeground {colorUseRule=UseColor_ColorTable, specifiedColor=<rgba 0 0 0 0>}
       normalPadding=<ltrb 8 1 8 1>
       pressed {
        bIsDynamicallyLoaded=false
        drawAs=RoundedBox
        imageType=NoImage
        mirroring=NoMirror
        resourceName=None
        resourceObject=null
        tiling=NoTile
        imageSize=<0 0>
        margin=<ltrb 0 0 0 0>
        outlineSettings {
         bUseBrushTransparency=false
         roundingType=FixedRadius
         width=1
         color {colorUseRule=UseColor_ColorTable, specifiedColor=<rgba 0 0 0 0>}
         cornerRadii=<4 4 4 4>
        }
        tintColor {colorUseRule=UseColor_Specified, specifiedColor=<rgba 0.0048 0.0048 0.0048 1>}
        uVRegion {bIsValid=false, max=<0 0>, min=<0 0>}
       }
       pressedForeground {colorUseRule=UseColor_ColorTable, specifiedColor=<rgba 0 0 0 0>}
       pressedPadding=<ltrb 8 2 8 0>
       pressedSlateSound {resourceObject=null}
      }
      contentPadding=<ltrb 0 0 0 0>
      downArrowImage {
       bIsDynamicallyLoaded=false
       drawAs=Image
       imageType=Vector
       mirroring=NoMirror
       resourceName=../../../Engine/Content/Slate/Starship/CoreWidgets/ComboBox/wide-chevron-down.svg
       resourceObject=null
       tiling=NoTile
       imageSize=<16 16>
       margin=<ltrb 0 0 0 0>
       outlineSettings {
        bUseBrushTransparency=false
        roundingType=HalfHeightRadius
        width=0
        color {colorUseRule=UseColor_Specified, specifiedColor=<rgba 0 0 0 0>}
        cornerRadii=<0 0 0 0>
       }
       tintColor {colorUseRule=UseColor_Specified, specifiedColor=<rgba 1 1 1 1>}
       uVRegion {bIsValid=false, max=<0 0>, min=<0 0>}
      }
      downArrowPadding=<ltrb 2 2 2 2>
      menuBorderBrush {
       bIsDynamicallyLoaded=false
       drawAs=RoundedBox
       imageType=NoImage
       mirroring=NoMirror
       resourceName=None
       resourceObject=null
       tiling=NoTile
       imageSize=<0 0>
       margin=<ltrb 0 0 0 0>
       outlineSettings {
        bUseBrushTransparency=false
        roundingType=FixedRadius
        width=1
        color {colorUseRule=UseColor_ColorTable, specifiedColor=<rgba 0 0 0 0>}
        cornerRadii=<0 0 0 0>
       }
       tintColor {colorUseRule=UseColor_Specified, specifiedColor=<rgba 0.0395 0.0395 0.0395 1>}
       uVRegion {bIsValid=false, max=<0 0>, min=<0 0>}
      }
      menuBorderPadding=<ltrb 0 0 0 0>
      shadowColorAndOpacity=<rgba 0 0 0 1>
      shadowOffset=<0 0>
     }
     contentPadding=<ltrb 0 0 0 0>
     menuRowPadding=<ltrb 8 3 8 3>
     pressedSlateSound {resourceObject=null}
     selectionChangeSlateSound {resourceObject=null}
    }
   }
   slot {class=VerticalBoxSlot}
  }
  ScrollBox_30 {
   widgetClassPath=&ScrollBox
   bIsVariable=true
   parent=AddHere
   widget=ScrollBox_30
   properties {
    widgetBarStyle {
     thickness=8
     draggedThumbImage {
      bIsDynamicallyLoaded=false
      drawAs=RoundedBox
      imageType=NoImage
      mirroring=NoMirror
      resourceName=None
      resourceObject=null
      tiling=NoTile
      imageSize=<0 0>
      margin=<ltrb 0 0 0 0>
      outlineSettings {
       bUseBrushTransparency=false
       roundingType=FixedRadius
       width=0
       color {colorUseRule=UseColor_Specified, specifiedColor=<rgba 0 0 0 0>}
       cornerRadii=<4 4 4 4>
      }
      tintColor {colorUseRule=UseColor_Specified, specifiedColor=<rgba 0.2159 0.2159 0.2159 1>}
      uVRegion {bIsValid=false, max=<0 0>, min=<0 0>}
     }
     horizontalBackgroundImage {
      bIsDynamicallyLoaded=false
      drawAs=NoDrawType
      imageType=NoImage
      mirroring=NoMirror
      resourceName=None
      resourceObject=null
      tiling=NoTile
      imageSize=<0 0>
      margin=<ltrb 0 0 0 0>
      outlineSettings {
       bUseBrushTransparency=false
       roundingType=HalfHeightRadius
       width=0
       color {colorUseRule=UseColor_Specified, specifiedColor=<rgba 0 0 0 0>}
       cornerRadii=<0 0 0 0>
      }
      tintColor {colorUseRule=UseColor_Specified, specifiedColor=<rgba 1 1 1 1>}
      uVRegion {bIsValid=false, max=<0 0>, min=<0 0>}
     }
     horizontalBottomSlotImage {
      bIsDynamicallyLoaded=false
      drawAs=NoDrawType
      imageType=NoImage
      mirroring=NoMirror
      resourceName=None
      resourceObject=null
      tiling=NoTile
      imageSize=<0 0>
      margin=<ltrb 0 0 0 0>
      outlineSettings {
       bUseBrushTransparency=false
       roundingType=HalfHeightRadius
       width=0
       color {colorUseRule=UseColor_Specified, specifiedColor=<rgba 0 0 0 0>}
       cornerRadii=<0 0 0 0>
      }
      tintColor {colorUseRule=UseColor_Specified, specifiedColor=<rgba 1 1 1 1>}
      uVRegion {bIsValid=false, max=<0 0>, min=<0 0>}
     }
     horizontalTopSlotImage {
      bIsDynamicallyLoaded=false
      drawAs=NoDrawType
      imageType=NoImage
      mirroring=NoMirror
      resourceName=None
      resourceObject=null
      tiling=NoTile
      imageSize=<0 0>
      margin=<ltrb 0 0 0 0>
      outlineSettings {
       bUseBrushTransparency=false
       roundingType=HalfHeightRadius
       width=0
       color {colorUseRule=UseColor_Specified, specifiedColor=<rgba 0 0 0 0>}
       cornerRadii=<0 0 0 0>
      }
      tintColor {colorUseRule=UseColor_Specified, specifiedColor=<rgba 1 1 1 1>}
      uVRegion {bIsValid=false, max=<0 0>, min=<0 0>}
     }
     hoveredThumbImage {
      bIsDynamicallyLoaded=false
      drawAs=RoundedBox
      imageType=NoImage
      mirroring=NoMirror
      resourceName=None
      resourceObject=null
      tiling=NoTile
      imageSize=<0 0>
      margin=<ltrb 0 0 0 0>
      outlineSettings {
       bUseBrushTransparency=false
       roundingType=FixedRadius
       width=0
       color {colorUseRule=UseColor_Specified, specifiedColor=<rgba 0 0 0 0>}
       cornerRadii=<4 4 4 4>
      }
      tintColor {colorUseRule=UseColor_Specified, specifiedColor=<rgba 0.2159 0.2159 0.2159 1>}
      uVRegion {bIsValid=false, max=<0 0>, min=<0 0>}
     }
     normalThumbImage {
      bIsDynamicallyLoaded=false
      drawAs=RoundedBox
      imageType=NoImage
      mirroring=NoMirror
      resourceName=None
      resourceObject=null
      tiling=NoTile
      imageSize=<0 0>
      margin=<ltrb 0 0 0 0>
      outlineSettings {
       bUseBrushTransparency=false
       roundingType=FixedRadius
       width=0
       color {colorUseRule=UseColor_Specified, specifiedColor=<rgba 0 0 0 0>}
       cornerRadii=<4 4 4 4>
      }
      tintColor {colorUseRule=UseColor_Specified, specifiedColor=<rgba 0.0953 0.0953 0.0953 1>}
      uVRegion {bIsValid=false, max=<0 0>, min=<0 0>}
     }
     verticalBackgroundImage {
      bIsDynamicallyLoaded=false
      drawAs=NoDrawType
      imageType=NoImage
      mirroring=NoMirror
      resourceName=None
      resourceObject=null
      tiling=NoTile
      imageSize=<0 0>
      margin=<ltrb 0 0 0 0>
      outlineSettings {
       bUseBrushTransparency=false
       roundingType=HalfHeightRadius
       width=0
       color {colorUseRule=UseColor_Specified, specifiedColor=<rgba 0 0 0 0>}
       cornerRadii=<0 0 0 0>
      }
      tintColor {colorUseRule=UseColor_Specified, specifiedColor=<rgba 1 1 1 1>}
      uVRegion {bIsValid=false, max=<0 0>, min=<0 0>}
     }
     verticalBottomSlotImage {
      bIsDynamicallyLoaded=false
      drawAs=NoDrawType
      imageType=NoImage
      mirroring=NoMirror
      resourceName=None
      resourceObject=null
      tiling=NoTile
      imageSize=<0 0>
      margin=<ltrb 0 0 0 0>
      outlineSettings {
       bUseBrushTransparency=false
       roundingType=HalfHeightRadius
       width=0
       color {colorUseRule=UseColor_Specified, specifiedColor=<rgba 0 0 0 0>}
       cornerRadii=<0 0 0 0>
      }
      tintColor {colorUseRule=UseColor_Specified, specifiedColor=<rgba 1 1 1 1>}
      uVRegion {bIsValid=false, max=<0 0>, min=<0 0>}
     }
     verticalTopSlotImage {
      bIsDynamicallyLoaded=false
      drawAs=NoDrawType
      imageType=NoImage
      mirroring=NoMirror
      resourceName=None
      resourceObject=null
      tiling=NoTile
      imageSize=<0 0>
      margin=<ltrb 0 0 0 0>
      outlineSettings {
       bUseBrushTransparency=false
       roundingType=HalfHeightRadius
       width=0
       color {colorUseRule=UseColor_Specified, specifiedColor=<rgba 0 0 0 0>}
       cornerRadii=<0 0 0 0>
      }
      tintColor {colorUseRule=UseColor_Specified, specifiedColor=<rgba 1 1 1 1>}
      uVRegion {bIsValid=false, max=<0 0>, min=<0 0>}
     }
    }
    widgetStyle {
     barThickness=8
     bottomShadowBrush {
      bIsDynamicallyLoaded=false
      drawAs=Box
      imageType=FullColor
      mirroring=NoMirror
      resourceName=../../../Engine/Content/Slate/Common/ScrollBoxShadowBottom.png
      resourceObject=null
      tiling=NoTile
      imageSize=<16 8>
      margin=<ltrb 0.5 0 0.5 1>
      outlineSettings {
       bUseBrushTransparency=false
       roundingType=HalfHeightRadius
       width=0
       color {colorUseRule=UseColor_Specified, specifiedColor=<rgba 0 0 0 0>}
       cornerRadii=<0 0 0 0>
      }
      tintColor {colorUseRule=UseColor_Specified, specifiedColor=<rgba 1 1 1 1>}
      uVRegion {bIsValid=false, max=<0 0>, min=<0 0>}
     }
     horizontalScrolledContentPadding=<ltrb 0 0 1 0>
     leftShadowBrush {
      bIsDynamicallyLoaded=false
      drawAs=Box
      imageType=FullColor
      mirroring=NoMirror
      resourceName=../../../Engine/Content/Slate/Common/ScrollBoxShadowLeft.png
      resourceObject=null
      tiling=NoTile
      imageSize=<8 16>
      margin=<ltrb 1 0.5 0 0.5>
      outlineSettings {
       bUseBrushTransparency=false
       roundingType=HalfHeightRadius
       width=0
       color {colorUseRule=UseColor_Specified, specifiedColor=<rgba 0 0 0 0>}
       cornerRadii=<0 0 0 0>
      }
      tintColor {colorUseRule=UseColor_Specified, specifiedColor=<rgba 1 1 1 1>}
      uVRegion {bIsValid=false, max=<0 0>, min=<0 0>}
     }
     rightShadowBrush {
      bIsDynamicallyLoaded=false
      drawAs=Box
      imageType=FullColor
      mirroring=NoMirror
      resourceName=../../../Engine/Content/Slate/Common/ScrollBoxShadowRight.png
      resourceObject=null
      tiling=NoTile
      imageSize=<8 16>
      margin=<ltrb 0 0.5 1 0.5>
      outlineSettings {
       bUseBrushTransparency=false
       roundingType=HalfHeightRadius
       width=0
       color {colorUseRule=UseColor_Specified, specifiedColor=<rgba 0 0 0 0>}
       cornerRadii=<0 0 0 0>
      }
      tintColor {colorUseRule=UseColor_Specified, specifiedColor=<rgba 1 1 1 1>}
      uVRegion {bIsValid=false, max=<0 0>, min=<0 0>}
     }
     topShadowBrush {
      bIsDynamicallyLoaded=false
      drawAs=Box
      imageType=FullColor
      mirroring=NoMirror
      resourceName=../../../Engine/Content/Slate/Common/ScrollBoxShadowTop.png
      resourceObject=null
      tiling=NoTile
      imageSize=<16 8>
      margin=<ltrb 0.5 1 0.5 0>
      outlineSettings {
       bUseBrushTransparency=false
       roundingType=HalfHeightRadius
       width=0
       color {colorUseRule=UseColor_Specified, specifiedColor=<rgba 0 0 0 0>}
       cornerRadii=<0 0 0 0>
      }
      tintColor {colorUseRule=UseColor_Specified, specifiedColor=<rgba 1 1 1 1>}
      uVRegion {bIsValid=false, max=<0 0>, min=<0 0>}
     }
     verticalScrolledContentPadding=<ltrb 0 0 0 1>
    }
   }
   slot {class=VerticalBoxSlot, properties {size {sizeRule=Fill, value=1}}}
  }
 ]
}
