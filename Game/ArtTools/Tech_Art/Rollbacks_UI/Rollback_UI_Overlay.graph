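# Tech-art debug overlay widget that hosts the Rollback UI panel with a close button above it.
# A full-screen vertical box stacks a generic tech-art button over the embedded RollbackUI widget, which fills the remaining space.
# Clicking close hides the mouse cursor, restores game-only input for the local player, runs the Fort.RollbackUI 0 console command, and removes the widget from its parent.
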WidgetBlueprint /Game/ArtTools/Tech_Art/Rollbacks_UI/Rollback_UI_Overlay
graphs [EventGraph {nodes [
   "On Clicked (Button_CloseWidget)"!() -> (OutputDelegate, then, widget:User Widget Object Reference, event:Pointer Event Structure) {is=title K2Node_ComponentBoundEvent}
   n1 = Get Local Player Controller From ID!(ControllerID:int=Get Local Player Controller ID!(Player:Player Controller Object Reference=GetPlayerController!(PlayerIndex:int=0) -> (ReturnValue:Player Controller Object Reference)) -> (ReturnValue:int=0) {is=title K2Node_CallFunction}) -> (ReturnValue:Player Controller Object Reference) {is=title K2Node_CallFunction}
   > Set bShowMouseCursor!(execute, bShowMouseCursor:bool=false, self:Player Controller Object Reference=@n1) -> (then, Output_Get:bool)
   > SetInputMode_GameOnly!(execute, PlayerController:Player Controller Object Reference=@n1, bFlushInput:bool=false) -> (then)
   > ExecuteConsoleCommand!(execute, Command=Fort.RollbackUI 0, SpecificPlayer) -> (then)
   > RemoveFromParent(self=Self-Reference())
   Event PreConstruct()
   Event Construct()
   Event Tick()
  ]}]
variables []
widgets {
 info {inheritedWidgetCount=0, parentClass=&UserWidget, rootWidgetClass=&Overlay, widgetCount=4}
 widgets [
  Overlay_61 {widgetClassPath=&Overlay, widget=Overlay_61}
  VerticalBox_0 {
   widgetClassPath=&VerticalBox
   parent=Overlay_61
   widget=VerticalBox_0
   slot {class=OverlaySlot, properties {horizontalAlignment=HAlign_Fill, verticalAlignment=VAlign_Fill}}
  }
  Button_CloseWidget {
   widgetClassPath=&TA_Generic01_Button
   parent=VerticalBox_0
   widget=Button_CloseWidget
   properties {label=Close the Widget, buttonColor_Default=<rgba 1 0.0289 0 1>}
   slot {class=VerticalBoxSlot}
  }
  RollbackUI {
   widgetClassPath=&/RollbackUI
   bIsVariable=true
   parent=VerticalBox_0
   widget=RollbackUI
   slot {class=VerticalBoxSlot, properties {size {sizeRule=Fill, value=1}}}
  }
 ]
}
